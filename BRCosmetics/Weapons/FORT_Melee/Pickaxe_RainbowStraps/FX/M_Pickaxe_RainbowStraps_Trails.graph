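# Translucent, two-sided unlit material for Niagara ribbon trails on the RainbowStraps pickaxe.
# Noise-distorted V coordinates index a curve-atlas color row, tinted by particle color and brightened toward one end.
# Opacity combines a sine cross-width mask with end fades, erodes it by noise, then applies particle alpha and view-angle fade.
# World position offset puffs the ribbon via a dynamic parameter and shifts it along the camera vector.
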Material /BRCosmetics/Weapons/FORT_Melee/Pickaxe_RainbowStraps/FX/M_Pickaxe_RainbowStraps_Trails
graph {
 properties {
  bUsedWithNiagaraRibbons=true
  blendMode=BLEND_Translucent
  pixelDepthOffsetMode=PDOM_Legacy
  shadingModel=MSM_Unlit
  twoSided=true
  usedShadingModels=Unlit
 }
 nodes [
  n1 = TextureCoordinate()
  n2 = Subtract(A=Sine(None=ComponentMask(None=@n1) {g=true}) {period=2}) {constB=0.1}
  n3 = DynamicParameter() {paramNames [WPO X, WPO Y, WPO Z, Param4]}
  n4 = MaterialFunctionCall(R=@n3.WPO X, G=@n3.WPO X, B=@n3.WPO X) {materialFunction=&AppendMany}
  Add(
   A=Multiply(A=Multiply(A=@n2, B=@n4.RGB), B=ScalarParameter() {parameterName=WPO Puff, sliderMax=30, sliderMin=-30})
   B=Multiply(
    A=MaterialFunctionCall() {materialFunction=&CameraDirectionVector}.Result
    B=ScalarParameter() {parameterName=Camera Offset}
   )
  )
  n5 = ComponentMask(None=TextureCoordinate()) {r=true}
  n6 = Multiply(A=Power(Base=OneMinus(None=@n5)) {constExponent=5}) {constB=3}
  n7 = ScalarParameter() {defaultValue=1, parameterName=Tiling X, sliderMax=2}
  n8 = ScalarParameter() {defaultValue=3, parameterName=Tiling Y, sliderMax=10}
  n9 = TextureSample(UVs=Multiply(A=TextureCoordinate(), B=AppendVector(A=@n7, B=@n8))) {samplerType=SAMPLERTYPE_Grayscale, texture=/FortUILibrary/UIKit/Buttons/Textures/LowResBlurredNoise}
  n10 = Subtract(A=@n9.R) {constB=0.5}
  n11 = ScalarParameter() {defaultValue=0.1, parameterName=Distortion Strength, sliderMax=1}
  n12 = ComponentMask(None=Add(A=TextureCoordinate(), B=Multiply(A=@n10, B=@n11) {constB=0.1})) {g=true}
  n13 = CurveAtlasRowParameter(CurveTime=@n12) {
   atlas=&/CA_Pickaxe_RainbowStraps
   curve=&/C_LinearColor_Pickaxe_RainbowStraps
  }
  n14 = ParticleColor()
  MaterialFunctionCall(Emissive=Multiply(A=Add(A=@n6), B=Multiply(A=@n13, B=@n14.RGB))) {materialFunction=&MF_FX_EmissiveAdjustments}
  n15 = TextureCoordinate()
  n16 = Sine(None=ComponentMask(None=@n15) {g=true}) {period=2}
  n17 = OneMinus(None=ComponentMask(None=@n15) {r=true})
  n18 = ScalarParameter() {defaultValue=0.2949, parameterName=Lead Edge Fade, sliderMax=1}
  n19 = Subtract(
   A=SmoothStep(Min=ScalarParameter() {parameterName=Trail End, sliderMax=1}, Value=@n17) {constMin=0.2}
   B=SmoothStep(Min=OneMinus(None=@n18), Value=@n17) {constMin=0.9}
  )
  n20 = Subtract(A=Multiply(A=Multiply(A=@n16, B=@n19)) {constB=5}, B=@n9.R)
  Multiply(
   A=Multiply(A=@n14.A, B=Saturate(None=@n20))
   B=MaterialFunctionCall("FadeSensitivity (optional)"=ScalarParameter() {defaultValue=1, parameterName="FadeSensitivity (optional) (S)", sliderMax=10}) {materialFunction=&FadeBasedOnViewAngle}.Result
  )
  TextureSampleParameter2D(UVs=@n12) {
   parameterName=Gradient
   samplerType=SAMPLERTYPE_LinearColor
   texture=&CA_Pickaxe_RainbowStraps
  }
 ]
}
